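# Sound cue for dropping the Paprika DPS shotgun weapon in the world.
# A Random node picks one of two drop recordings with equal weight, then a Modulator slightly varies pitch and volume.
# It plays through weapon-drop attenuation and the loot-collect sound class, with a weapon-drop concurrency set limiting overlapping plays.
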SoundCue /PaprikaCoreWeapons/Items/Weapons/PaprikaShotgun_DPS/Sounds/Paprika_Shotgun_Drop_Cue
root=Modulator_0
nodes [
 {id=Modulator_0, children [Random_1], properties {pitchMax=1, pitchMin=0.9, volumeMax=0.9, volumeMin=0.8}}
 {id=Random_1, children [WavePlayer_0,WavePlayer_4], properties {weights [1,1]}}
 {id=WavePlayer_0, properties {soundWaveAssetPtr=&/Paprika_Shotgun_Drop_01}}
 {id=WavePlayer_4, properties {soundWaveAssetPtr=&/Paprika_Shotgun_Drop_02}}
]
properties {
 attenuationSettings=&Fort_Weapon_Drop_Att
 maxDistance=3000
 soundClassObject=&Fort_Loot_Collect
 volumeMultiplier=0.6
 concurrencyOverrides {resolutionRule=PreventNew}
 concurrencySet [&Weapon_Drop_StopOldest]
}
