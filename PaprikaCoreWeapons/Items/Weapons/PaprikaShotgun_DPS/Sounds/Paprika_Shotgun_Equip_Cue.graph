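# Sound Cue that plays the equip sound for the Paprika DPS shotgun, using one of several recorded variations.
# A Quality Level node routes its first two slots to a random pick of four waves and its last two to a pick of two, reducing variation.
# A Modulator at the root adds slight volume variance.
# The cue borrows weapon-reload attenuation, the loot-collect sound class and a pickup concurrency set.
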SoundCue /PaprikaCoreWeapons/Items/Weapons/PaprikaShotgun_DPS/Sounds/Paprika_Shotgun_Equip_Cue
root=Modulator_1
nodes [
 {id=Modulator_1, children [QualityLevel_0], properties {volumeMax=1.1, volumeMin=1}}
 {id=QualityLevel_0, children [Random_0,Random_0,Random_1,Random_1]}
 {id=Random_0, children [WavePlayer_2,WavePlayer_3,WavePlayer_4,WavePlayer_5], properties {weights [1,1,1,1]}}
 {id=Random_1, children [WavePlayer_4,WavePlayer_5], properties {weights [1,1]}}
 {id=WavePlayer_2, properties {soundWaveAssetPtr=&/Paprika_Shotgun_Equip_04}}
 {id=WavePlayer_3, properties {soundWaveAssetPtr=&/Paprika_Shotgun_Equip_03}}
 {id=WavePlayer_4, properties {soundWaveAssetPtr=&/Paprika_Shotgun_Equip_02}}
 {id=WavePlayer_5, properties {soundWaveAssetPtr=&/Paprika_Shotgun_Equip_01}}
]
properties {
 attenuationSettings=&Fort_Weapon_Reload_Att
 maxDistance=2250
 soundClassObject=&Fort_Loot_Collect
 volumeMultiplier=0.6
 concurrencyOverrides {maxCount=1, resolutionRule=PreventNew}
 concurrencySet [&PickupMaximum_Oldest]
}
