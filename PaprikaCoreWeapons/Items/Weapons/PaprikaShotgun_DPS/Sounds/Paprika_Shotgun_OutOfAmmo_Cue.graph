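# Sound cue for the out-of-ammo fire sound of the Paprika DPS shotgun, built from waves named as first-person sounds.
# A Random node picks between two equally weighted wave variants and feeds a Modulator at the root for playback variation.
# The cue uses the weapon foley sound class and the shared weapon reload attenuation, with boosted volume, a concurrency override and the debug flag enabled.
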SoundCue /PaprikaCoreWeapons/Items/Weapons/PaprikaShotgun_DPS/Sounds/Paprika_Shotgun_OutOfAmmo_Cue
root=Modulator_0
nodes [
 {id=Modulator_0, children [Random_0]}
 {id=Random_0, children [WavePlayer_1,WavePlayer_2], properties {weights [1,1]}}
 {
  id=WavePlayer_1
  properties {soundWaveAssetPtr=&/Paprika_SG_CoreDPS_Fire_1P_OutOfAmmo_01}
 }
 {
  id=WavePlayer_2
  properties {soundWaveAssetPtr=&/Paprika_SG_CoreDPS_Fire_1P_OutOfAmmo_02}
 }
]
properties {
 attenuationSettings=&Fort_Weapon_Reload_Att
 bDebug=true
 bOverrideConcurrency=true
 maxDistance=2250
 soundClassObject=&Fort_WeaponFoley
 volumeMultiplier=1.344
}
