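# Sound cue for the cartridge eject step of the Paprika DPS shotgun reload, playing a magazine-out foley sample.
# A single wave player feeds a modulator at the root, with lowered cue volume.
# It uses the Fort weapon reload attenuation and weapon foley sound class, overrides concurrency, and has debug enabled.
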SoundCue /PaprikaCoreWeapons/Items/Weapons/PaprikaShotgun_DPS/Sounds/Paprika_Shotgun_Reload_Cartridge_Eject_Cue
root=Modulator_2
nodes [
 {id=Modulator_2, children [WavePlayer_0]}
 {
  id=WavePlayer_0
  properties {soundWaveAssetPtr=&/Paprika_SG_CoreDPS_Reload_Mag_Out_01}
 }
]
properties {
 attenuationSettings=&Fort_Weapon_Reload_Att
 bDebug=true
 bOverrideConcurrency=true
 maxDistance=2250
 soundClassObject=&Fort_WeaponFoley
 volumeMultiplier=0.351
}
