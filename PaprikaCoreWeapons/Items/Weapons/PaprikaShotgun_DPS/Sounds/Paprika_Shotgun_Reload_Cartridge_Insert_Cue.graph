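# Sound cue for the Paprika DPS shotgun's reload foley, heard when a cartridge is inserted.
# A single mag-in wave passes through a Modulator, which randomizes volume and pitch, into a Mixer whose second input is empty.
# The cue uses the weapon foley sound class and shared reload attenuation, lowers overall volume, and overrides concurrency.
# Its debug flag is left on.
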SoundCue /PaprikaCoreWeapons/Items/Weapons/PaprikaShotgun_DPS/Sounds/Paprika_Shotgun_Reload_Cartridge_Insert_Cue
root=Mixer_0
nodes [
 {id=Mixer_0, children [Modulator_0,null], properties {inputVolume [1,1]}}
 {id=Modulator_0, children [WavePlayer_3], properties {pitchMin=1, volumeMax=0.6, volumeMin=0.5}}
 {
  id=WavePlayer_3
  properties {soundWaveAssetPtr=&/Paprika_SG_CoreDPS_Reload_Mag_In_01}
 }
]
properties {
 attenuationSettings=&Fort_Weapon_Reload_Att
 bDebug=true
 bOverrideConcurrency=true
 maxDistance=2250
 soundClassObject=&Fort_WeaponFoley
 volumeMultiplier=0.351
}
