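# Sound cue for the cartridge pouch foley during the Paprika DPS shotgun reload.
# A single wave player plays a mag-hit reload wave through a modulator left at its default settings.
# It uses the shared Fortnite weapon-reload attenuation and weapon foley sound class at reduced volume, overrides concurrency, and has its debug flag left on.
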SoundCue /PaprikaCoreWeapons/Items/Weapons/PaprikaShotgun_DPS/Sounds/Paprika_Shotgun_Reload_Cartridge_Pouch_Cue
root=Modulator_1
nodes [
 {id=Modulator_1, children [WavePlayer_4]}
 {
  id=WavePlayer_4
  properties {soundWaveAssetPtr=&/Paprika_SG_CoreDPS_Reload_Mag_Hit_01}
 }
]
properties {
 attenuationSettings=&Fort_Weapon_Reload_Att
 bDebug=true
 bOverrideConcurrency=true
 maxDistance=2250
 soundClassObject=&Fort_WeaponFoley
 volumeMultiplier=0.401
}
